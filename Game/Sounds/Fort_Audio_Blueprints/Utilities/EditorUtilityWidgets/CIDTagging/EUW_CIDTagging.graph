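# Editor utility widget for audio authors to set SoundLibrary.ID foley tags (step, clothing, hand) and notes on character item definitions (CIDs).
# Apply Tags strips existing SoundLibrary.ID tags, appends tags requested by name from the selected enums, then saves the CID.
# Add Note saves the note as CIDTagging metadata.
# Generate CSV loads all Athena character assets and copies an asset-action foley-tag export to the clipboard via inline Python.
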EditorUtilityWidgetBlueprint /Game/Sounds/Fort_Audio_Blueprints/Utilities/EditorUtilityWidgets/CIDTagging/EUW_CIDTagging
graphs [
 EventGraph {nodes [
   "On Clicked (ApplyTags)"() {is=title K2Node_ComponentBoundEvent}
   > n1 = GetItemData!(execute, self:Item Definition Base Object Reference=Get CID() -> (CID:Athena Character Item Definition Object Reference)) -> (Found:Exec, NotFound:Exec, ItemData:"Owned Gameplay Tags Item Data Structure (by ref)")
   > n2 = Set Tags(Tags:Gameplay Tag Container Structure=Break Owned Gameplay Tags Item Data!(ItemComponentData_OwnedGameplayTags:"Owned Gameplay Tags Item Data Structure (by ref)"=@n1) -> (Tags:Gameplay Tag Container Structure) {is=title K2Node_BreakStruct}) -> (Output_Get:Gameplay Tag Container Structure)
   > n3 = RemoveSoundLibIDTags!(execute, self, In Tag Container:"Gameplay Tag Container Structure (by ref)"=@n2) -> (then, Out Tag Container:Gameplay Tag Container Structure)
   > n4 = Sequence()
   > AppendGameplayTagContainers!(
    execute
    InOutTagContainer:"Gameplay Tag Container Structure (by ref)"=@n3
    InTagContainer:"Gameplay Tag Container Structure (by ref)"=RequestSoundLibTags!(
     StepTypeTag:CID_TagType_Step Enum=Get StepTypeTag() -> (StepTypeTag:CID_TagType_Step Enum=NewEnumerator3)
     ClothingTypeTag:CID_TagType_Clothing Enum=Get ClothingTypeTag() -> (ClothingTypeTag:CID_TagType_Clothing Enum=NewEnumerator0)
     HandTypeTag:CID_TagType_Hand Enum=Get HandTypeTag() -> (HandTypeTag:CID_TagType_Hand Enum=NewEnumerator0)
    ) -> (Tags:Gameplay Tag Container Structure)
   ) -> (then)
   SetItemData!(
    execute=@n4.then_1
    self:Item Definition Base Object Reference=Get CID()
    ItemData:Owned Gameplay Tags Item Data Structure=Make Owned Gameplay Tags Item Data!(Tags:Gameplay Tag Container Structure=@n3) -> (ItemComponentData_OwnedGameplayTags:Owned Gameplay Tags Item Data Structure) {is=title K2Node_MakeStruct}
   ) -> (then)
   > SaveLoadedAsset(AssetToSave=Get CID(), bOnlyIfIsDirty=true)
   "On Clicked (GenerateCSV)"() {is=title K2Node_ComponentBoundEvent}
   > n5 = GetAssetsByPath!(
    execute
    self:Asset Registry Interface=GetAssetRegistry()
    PackagePath:Name=/BRCosmetics/Athena/Items/Cosmetics/Characters/
    bRecursive:bool=true
    bIncludeOnlyOnDiskAssets:bool=false
   ) -> (then, OutAssetData:Array of Asset Data Structures, ReturnValue:bool)
   > Clear(TargetArray:Array of Object References=Get Objects() -> (Objects:Array of Object References))
   > Clear(TargetArray:Array of Strings=Get ObjectNames() -> (ObjectNames:Array of Strings))
   > n6 = For Each Loop(Array:Array of Asset Data Structures=@n5) -> (Array Element:Asset Data Structure)
   n7 = ExportFoleyTagsToCSVString!(
    execute=@n6.Completed
    self:Character Item Definition Asset Actions Object Reference=GetOrCreateAssetActions!() -> (AssetActionInstance:Character Item Definition Asset Actions Object Reference)
    Objects:Array of Object References=Get Objects()
   ) -> (then, CSV Output:Text)
   > Execute Python Script!(
    execute
    PythonScript="import os\r\nfrom tkinter import *\r\nroot = Tk()\r\nroot.clipboard_clear()\r\n\r\ncsv = csv.replace(\"\\n\", \"\")\r\n\r\nroot.clipboard_append(csv)\r\nprint('CSV copied to clipboard')\r\nroot.withdraw()"
    csv:String="To String (Text)"(InText=@n7)
   ) -> (then, ReturnValue) {inputs [csv]}
   Add(execute=@n6.LoopBody, TargetArray:Array of Object References=Get Objects(), NewItem:"Object Reference (by ref)"=GetAsset(InAssetData=@n6)) {is=title K2Node_CallArrayFunction}
   "On Clicked (ApplyNotes)"() {is=title K2Node_ComponentBoundEvent}
   n8 = Get CID()
   > SetMetadataTag!(execute, Object:Object Reference=@n8, Tag:Name=CIDTagging, Value:String=Get Notes() -> (Notes:String)) -> (then)
   > SaveLoadedAsset(AssetToSave=@n8, bOnlyIfIsDirty=true)
   Event Construct()
   n9 = Get StepTypeProperty() -> (StepTypeProperty:Single Property View Object Reference)
   > SetObject(self=@n9, NewObject=Self-Reference())
   > SetPropertyName!(execute, self:Single Property View Object Reference=@n9, NewPropertyName:Name=StepTypeTag) -> (then)
   n10 = Get ClothingTypeProperty() -> (ClothingTypeProperty:Single Property View Object Reference)
   > SetObject(self=@n10, NewObject=Self-Reference())
   > SetPropertyName(self=@n10, NewPropertyName=ClothingTypeTag)
   n11 = Get HandTypeProperty() -> (HandTypeProperty:Single Property View Object Reference)
   > SetObject(self=@n11, NewObject=Self-Reference())
   > SetPropertyName(self=@n11, NewPropertyName=HandTypeTag)
   n12 = Get CIDProperty() -> (CIDProperty:Single Property View Object Reference)
   > SetObject(self=@n12, NewObject=Self-Reference())
   > SetPropertyName(self=@n12, NewPropertyName=CID)
   n13 = Get NotesTypeProperty() -> (NotesTypeProperty:Single Property View Object Reference)
   > SetObject(self=@n13, NewObject=Self-Reference())
   > SetPropertyName(self=@n13, NewPropertyName=Notes)
   Event PreConstruct()
  ]}
 GetOrCreateAssetActions {nodes [
   GetOrCreateAssetActions() {is=title K2Node_FunctionEntry}
   > n1 = Get!(execute:Exec) -> (then:Exec, else:Exec, AssetActionInstance:Character Item Definition Asset Actions Object Reference)
   Return Node!(
    execute=[@n1.then, @n3.then]
    AssetActionInstance:Character Item Definition Asset Actions Object Reference=Get AssetActionInstance() -> (AssetActionInstance:Character Item Definition Asset Actions Object Reference)
   )
   n2 = Construct Object from Class(
    execute=@n1.else
    Class=/Game/Sounds/Fort_Audio_Blueprints/Utilities/AssetActions/CharacterItemDefinitionAssetActions.CharacterItemDefinitionAssetActions_C
   ) -> (ReturnValue:Character Item Definition Asset Actions Object Reference)
   > n3 = Set AssetActionInstance(AssetActionInstance:Character Item Definition Asset Actions Object Reference=@n2) -> (Output_Get:Character Item Definition Asset Actions Object Reference)
  ]}
 ObjectAlreadyInArray {nodes [
   n1 = ObjectAlreadyInArray!() -> (then, Object:Object Reference) {is=title K2Node_FunctionEntry}
   > n2 = For Each Loop(Array:Array of Object References=Get Objects()) -> (Array Element:Object Reference)
   Return Node!(execute=@n2.Completed, Result:bool=false)
   > Branch(Condition="Equal Exactly (String)"(A=GetDisplayName(Object=@n2), B=GetDisplayName(Object=@n1)))
   > Return Node(Result=true)
  ]}
 RemoveSoundLibIDTags {nodes [
   n1 = RemoveSoundLibIDTags!() -> (then, In Tag Container:"Gameplay Tag Container Structure (by ref)") {is=title K2Node_FunctionEntry}
   > n2 = For Each Loop(Array:Array of Gameplay Tag Structures=BreakGameplayTagContainer!(GameplayTagContainer:"Gameplay Tag Container Structure (by ref)"=@n1) -> (GameplayTags:Array of Gameplay Tag Structures)) -> (Array Element:Gameplay Tag Structure)
   n3 = For Each Loop(Exec=@n2.Completed, Array=Get TagsToRemove() -> (TagsToRemove:Array of Gameplay Tag Structures))
   Return Node!(execute=@n3.Completed, Out Tag Container:"Gameplay Tag Container Structure (by ref)"=@n1)
   > RemoveGameplayTag!(execute, TagContainer:"Gameplay Tag Container Structure (by ref)"=@n1, Tag:Gameplay Tag Structure=@n3) -> (then, ReturnValue:bool)
   Branch(
    execute=@n2.LoopBody
    Condition=MatchesTag!(TagOne:Gameplay Tag Structure=@n2, TagTwo:Gameplay Tag Structure=(TagName="SoundLibrary.ID"), bExactMatch:bool=false) -> (ReturnValue:bool)
   )
   > Add(TargetArray:Array of Gameplay Tag Structures=Get TagsToRemove(), NewItem:"Gameplay Tag Structure (by ref)"=@n2)
  ]}
 RequestSoundLibTags {nodes [
   RequestSoundLibTags!() -> (
    then
    StepTypeTag:CID_TagType_Step Enum=NewEnumerator3
    ClothingTypeTag:CID_TagType_Clothing Enum=NewEnumerator0
    HandTypeTag:CID_TagType_Hand Enum=NewEnumerator0
   ) {is=title K2Node_FunctionEntry}
   > n1 = Sequence!(execute) -> (then_0, then_1, then_2:Exec, then_3:Exec)
   n2 = Get StepTypeTag() -> (StepTypeTag=NewEnumerator3)
   > n3 = RequestGameplayTag!(
    execute
    TagName:String=Append(A=SoundLibrary.ID.StepType., B=Enum to String!(Enumerator:CID_TagType_Step Enum=@n2) -> (ReturnValue:String))
    bErrorIfNotFound:bool=true
   ) -> (then, OutTag:Gameplay Tag Structure, ReturnValue:bool)
   > AddGameplayTag!(
    execute
    TagContainer:"Gameplay Tag Container Structure (by ref)"=Get TagContainer() -> (TagContainer:Gameplay Tag Container Structure)
    Tag:Gameplay Tag Structure=@n3
   ) -> (then)
   Return Node!(execute=@n1.then_3, Tags:Gameplay Tag Container Structure=Get TagContainer())
   n4 = Get ClothingTypeTag() -> (ClothingTypeTag=NewEnumerator0)
   n5 = RequestGameplayTag(
    execute=@n1.then_1
    TagName=Append(A=SoundLibrary.ID.ClothingType., B=Enum to String(Enumerator:CID_TagType_Clothing Enum=@n4))
    bErrorIfNotFound=true
   )
   > AddGameplayTag(TagContainer=Get TagContainer(), Tag=@n5)
   n6 = Get HandTypeTag() -> (HandTypeTag=NewEnumerator0)
   n7 = RequestGameplayTag(
    execute=@n1.then_2
    TagName=Append(A=SoundLibrary.ID.HandType., B=Enum to String(Enumerator:CID_TagType_Hand Enum=@n6))
    bErrorIfNotFound=true
   )
   > AddGameplayTag(TagContainer=Get TagContainer(), Tag=@n7)
  ]}
]
variables [
 AssetActionInstance {type=CharacterItemDefinitionAssetActions_C, default=null}
 CID {type=AthenaCharacterItemDefinition, default=null}
 ClothingTypeTag {type=CID_TagType_Clothing, default=Generic}
 data {type=ItemComponentData_OwnedGameplayTags, default {tags {gameplayTags []}}}
 HandTypeTag {type=CID_TagType_Hand, default=Bare}
 Notes {type=string, default=""}
 ObjectNames {type="array<string>", default []}
 Objects {type="array<Object>", default []}
 StepTypeTag {type=CID_TagType_Step, default=BootsCombat}
 StepTypeTagMap {default {BootsCombat {tagName=None}, Sandal {tagName=SoundLibrary.ID.StepType.Sandal}}}
 Tags {type=GameplayTagContainer, default {gameplayTags []}}
]
widgets {
 info {rootWidgetClass=&CanvasPanel, widgetCount=20}
 widgets [
  CanvasPanel_0 {widgetClassPath=&CanvasPanel, widget=CanvasPanel_0}
  ScaleBox_262 {
   widgetClassPath=&ScaleBox
   parent=CanvasPanel_0
   widget=ScaleBox_262
   slot {class=CanvasPanelSlot, properties {layoutData {alignment=<0 0>, anchors {maximum=<0 0>, minimum=<0 0>}, offsets=<ltrb 5 5 600 400>}}}
  }
  VerticalBox_202 {
   widgetClassPath=&VerticalBox
   parent=ScaleBox_262
   widget=VerticalBox_202
   slot {class=ScaleBoxSlot, properties {horizontalAlignment=HAlign_Fill, verticalAlignment=VAlign_Fill}}
  }
  HorizontalBox_107 {
   widgetClassPath=&HorizontalBox
   parent=VerticalBox_202
   widget=HorizontalBox_107
   slot {class=VerticalBoxSlot, properties {size {sizeRule=Fill, value=0.8}}}
  }
  VerticalBox_208 {
   widgetClassPath=&VerticalBox
   parent=HorizontalBox_107
   widget=VerticalBox_208
   slot {class=HorizontalBoxSlot, properties {size {sizeRule=Fill, value=0.2}}}
  }
  CIDProperty {
   widgetClassPath=&SinglePropertyView
   parent=VerticalBox_208
   widget=CIDProperty
   properties {nameOverride=CID, propertyName=CID}
   slot {class=VerticalBoxSlot, properties {size {sizeRule=Fill, value=1.5}}}
  }
  StepTypeProperty {
   widgetClassPath=&SinglePropertyView
   parent=VerticalBox_208
   widget=StepTypeProperty
   properties {nameOverride=Step}
   slot {class=VerticalBoxSlot, properties {size {sizeRule=Fill, value=1}}}
  }
  ClothingTypeProperty {
   widgetClassPath=&SinglePropertyView
   parent=VerticalBox_208
   widget=ClothingTypeProperty
   properties {nameOverride=Clothing}
   slot {class=VerticalBoxSlot, properties {size {sizeRule=Fill, value=1}}}
  }
  HandTypeProperty {
   widgetClassPath=&SinglePropertyView
   parent=VerticalBox_208
   widget=HandTypeProperty
   properties {nameOverride=Hand}
   slot {class=VerticalBoxSlot, properties {size {sizeRule=Fill, value=1}}}
  }
  NotesTypeProperty {
   widgetClassPath=&SinglePropertyView
   parent=VerticalBox_208
   widget=NotesTypeProperty
   properties {nameOverride=Notes, propertyName=Notes}
   slot {class=VerticalBoxSlot, properties {size {sizeRule=Fill, value=1}}}
  }
  HorizontalBox_260 {
   widgetClassPath=&HorizontalBox
   parent=VerticalBox_202
   widget=HorizontalBox_260
   slot {class=VerticalBoxSlot, properties {size {sizeRule=Automatic, value=0.2}}}
  }
  ApplyTags {
   widgetClassPath=&Button
   parent=HorizontalBox_260
   widget=ApplyTags
   properties {widgetStyle {
     clickedSlateSound {resourceObject=null}
     disabled {
      bIsDynamicallyLoaded=false
      drawAs=RoundedBox
      imageType=NoImage
      mirroring=NoMirror
      resourceName=None
      resourceObject=null
      tiling=NoTile
      imageSize=<0 0>
      margin=<ltrb 0 0 0 0>
      outlineSettings {
       bUseBrushTransparency=false
       roundingType=FixedRadius
       width=1
       color {colorUseRule=UseColor_ColorTable, specifiedColor=<rgba 0 0 0 0>}
       cornerRadii=<4 4 4 4>
      }
      tintColor {colorUseRule=UseColor_Specified, specifiedColor=<rgba 0.0395 0.0395 0.0395 1>}
      uVRegion {bIsValid=false, max=<0 0>, min=<0 0>}
     }
     disabledForeground {colorUseRule=UseColor_ColorTable, specifiedColor=<rgba 0 0 0 0>}
     hovered {
      bIsDynamicallyLoaded=false
      drawAs=RoundedBox
      imageType=NoImage
      mirroring=NoMirror
      resourceName=None
      resourceObject=null
      tiling=NoTile
      imageSize=<0 0>
      margin=<ltrb 0 0 0 0>
      outlineSettings {
       bUseBrushTransparency=false
       roundingType=FixedRadius
       width=1
       color {colorUseRule=UseColor_ColorTable, specifiedColor=<rgba 0 0 0 0>}
       cornerRadii=<4 4 4 4>
      }
      tintColor {colorUseRule=UseColor_Specified, specifiedColor=<rgba 0.0953 0.0953 0.0953 1>}
      uVRegion {bIsValid=false, max=<0 0>, min=<0 0>}
     }
     hoveredForeground {colorUseRule=UseColor_ColorTable, specifiedColor=<rgba 0 0 0 0>}
     hoveredSlateSound {resourceObject=null}
     normal {
      bIsDynamicallyLoaded=false
      drawAs=RoundedBox
      imageType=NoImage
      mirroring=NoMirror
      resourceName=None
      resourceObject=null
      tiling=NoTile
      imageSize=<0 0>
      margin=<ltrb 0 0 0 0>
      outlineSettings {
       bUseBrushTransparency=false
       roundingType=FixedRadius
       width=1
       color {colorUseRule=UseColor_ColorTable, specifiedColor=<rgba 0 0 0 0>}
       cornerRadii=<4 4 4 4>
      }
      tintColor {colorUseRule=UseColor_Specified, specifiedColor=<rgba 0.0395 0.0395 0.0395 1>}
      uVRegion {bIsValid=false, max=<0 0>, min=<0 0>}
     }
     normalForeground {colorUseRule=UseColor_ColorTable, specifiedColor=<rgba 0 0 0 0>}
     normalPadding=<ltrb 12 1.5 12 1.5>
     pressed {
      bIsDynamicallyLoaded=false
      drawAs=RoundedBox
      imageType=NoImage
      mirroring=NoMirror
      resourceName=None
      resourceObject=null
      tiling=NoTile
      imageSize=<0 0>
      margin=<ltrb 0 0 0 0>
      outlineSettings {
       bUseBrushTransparency=false
       roundingType=FixedRadius
       width=1
       color {colorUseRule=UseColor_ColorTable, specifiedColor=<rgba 0 0 0 0>}
       cornerRadii=<4 4 4 4>
      }
      tintColor {colorUseRule=UseColor_Specified, specifiedColor=<rgba 0.0284 0.0284 0.0284 1>}
      uVRegion {bIsValid=false, max=<0 0>, min=<0 0>}
     }
     pressedForeground {colorUseRule=UseColor_ColorTable, specifiedColor=<rgba 0 0 0 0>}
     pressedPadding=<ltrb 12 2 12 0>
     pressedSlateSound {resourceObject=null}
    }}
   slot {class=HorizontalBoxSlot, properties {horizontalAlignment=HAlign_Left, size {sizeRule=Automatic, value=0.5}}}
  }
  ScaleBox_494 {
   widgetClassPath=&ScaleBox
   parent=ApplyTags
   widget=ScaleBox_494
   slot {class=ButtonSlot, properties {horizontalAlignment=HAlign_Fill, verticalAlignment=VAlign_Fill, padding=<ltrb 0 0 0 0>}}
  }
  CommonTextBlock_750 {
   widgetClassPath=&CommonTextBlock
   parent=ScaleBox_494
   widget=CommonTextBlock_750
   properties {
    style=&TextStyle-Base_C
    text=Apply Tags
    font {
     bForceMonospaced=false
     bMaterialIsStencil=false
     fontMaterial=null
     fontObject=&NotoSans
     letterSpacing=0
     monospacedWidth=1
     size=24
     skewAmount=0
     typefaceFontName=NotoSans-Regular
     outlineSettings {
      bApplyOutlineToDropShadows=false
      bMiteredCorners=false
      bSeparateFillAlpha=false
      outlineMaterial=null
      outlineSize=0
      outlineColor=<rgba 0 0 0 1>
     }
    }
    renderTransform {angle=0, scale=<1 1>, shear=<0 0>, translation=<0 -2.5>}
    shadowOffset=<0 0>
   }
   slot {class=ScaleBoxSlot}
  }
  ApplyNotes {
   widgetClassPath=&Button
   parent=HorizontalBox_260
   widget=ApplyNotes
   properties {widgetStyle {
     clickedSlateSound {resourceObject=null}
     disabled {
      bIsDynamicallyLoaded=false
      drawAs=RoundedBox
      imageType=NoImage
      mirroring=NoMirror
      resourceName=None
      resourceObject=null
      tiling=NoTile
      imageSize=<0 0>
      margin=<ltrb 0 0 0 0>
      outlineSettings {
       bUseBrushTransparency=false
       roundingType=FixedRadius
       width=1
       color {colorUseRule=UseColor_ColorTable, specifiedColor=<rgba 0 0 0 0>}
       cornerRadii=<4 4 4 4>
      }
      tintColor {colorUseRule=UseColor_Specified, specifiedColor=<rgba 0.0395 0.0395 0.0395 1>}
      uVRegion {bIsValid=false, max=<0 0>, min=<0 0>}
     }
     disabledForeground {colorUseRule=UseColor_ColorTable, specifiedColor=<rgba 0 0 0 0>}
     hovered {
      bIsDynamicallyLoaded=false
      drawAs=RoundedBox
      imageType=NoImage
      mirroring=NoMirror
      resourceName=None
      resourceObject=null
      tiling=NoTile
      imageSize=<0 0>
      margin=<ltrb 0 0 0 0>
      outlineSettings {
       bUseBrushTransparency=false
       roundingType=FixedRadius
       width=1
       color {colorUseRule=UseColor_ColorTable, specifiedColor=<rgba 0 0 0 0>}
       cornerRadii=<4 4 4 4>
      }
      tintColor {colorUseRule=UseColor_Specified, specifiedColor=<rgba 0.0953 0.0953 0.0953 1>}
      uVRegion {bIsValid=false, max=<0 0>, min=<0 0>}
     }
     hoveredForeground {colorUseRule=UseColor_ColorTable, specifiedColor=<rgba 0 0 0 0>}
     hoveredSlateSound {resourceObject=null}
     normal {
      bIsDynamicallyLoaded=false
      drawAs=RoundedBox
      imageType=NoImage
      mirroring=NoMirror
      resourceName=None
      resourceObject=null
      tiling=NoTile
      imageSize=<0 0>
      margin=<ltrb 0 0 0 0>
      outlineSettings {
       bUseBrushTransparency=false
       roundingType=FixedRadius
       width=1
       color {colorUseRule=UseColor_ColorTable, specifiedColor=<rgba 0 0 0 0>}
       cornerRadii=<4 4 4 4>
      }
      tintColor {colorUseRule=UseColor_Specified, specifiedColor=<rgba 0.0395 0.0395 0.0395 1>}
      uVRegion {bIsValid=false, max=<0 0>, min=<0 0>}
     }
     normalForeground {colorUseRule=UseColor_ColorTable, specifiedColor=<rgba 0 0 0 0>}
     normalPadding=<ltrb 12 1.5 12 1.5>
     pressed {
      bIsDynamicallyLoaded=false
      drawAs=RoundedBox
      imageType=NoImage
      mirroring=NoMirror
      resourceName=None
      resourceObject=null
      tiling=NoTile
      imageSize=<0 0>
      margin=<ltrb 0 0 0 0>
      outlineSettings {
       bUseBrushTransparency=false
       roundingType=FixedRadius
       width=1
       color {colorUseRule=UseColor_ColorTable, specifiedColor=<rgba 0 0 0 0>}
       cornerRadii=<4 4 4 4>
      }
      tintColor {colorUseRule=UseColor_Specified, specifiedColor=<rgba 0.0284 0.0284 0.0284 1>}
      uVRegion {bIsValid=false, max=<0 0>, min=<0 0>}
     }
     pressedForeground {colorUseRule=UseColor_ColorTable, specifiedColor=<rgba 0 0 0 0>}
     pressedPadding=<ltrb 12 2 12 0>
     pressedSlateSound {resourceObject=null}
    }}
   slot {class=HorizontalBoxSlot, properties {horizontalAlignment=HAlign_Left, size {sizeRule=Automatic, value=0.5}}}
  }
  ScaleBox_1 {
   widgetClassPath=&ScaleBox
   parent=ApplyNotes
   widget=ScaleBox_1
   slot {class=ButtonSlot, properties {horizontalAlignment=HAlign_Fill, verticalAlignment=VAlign_Fill, padding=<ltrb 0 0 0 0>}}
  }
  CommonTextBlock_1 {
   widgetClassPath=&CommonTextBlock
   parent=ScaleBox_1
   widget=CommonTextBlock_1
   properties {
    style=&TextStyle-Base_C
    text=Add Note
    font {
     bForceMonospaced=false
     bMaterialIsStencil=false
     fontMaterial=null
     fontObject=&NotoSans
     letterSpacing=0
     monospacedWidth=1
     size=24
     skewAmount=0
     typefaceFontName=NotoSans-Regular
     outlineSettings {
      bApplyOutlineToDropShadows=false
      bMiteredCorners=false
      bSeparateFillAlpha=false
      outlineMaterial=null
      outlineSize=0
      outlineColor=<rgba 0 0 0 1>
     }
    }
    renderTransform {angle=0, scale=<1 1>, shear=<0 0>, translation=<0 -2.5>}
    shadowOffset=<0 0>
   }
   slot {class=ScaleBoxSlot}
  }
  GenerateCSV {
   widgetClassPath=&Button
   parent=HorizontalBox_260
   widget=GenerateCSV
   properties {widgetStyle {
     clickedSlateSound {resourceObject=null}
     disabled {
      bIsDynamicallyLoaded=false
      drawAs=RoundedBox
      imageType=NoImage
      mirroring=NoMirror
      resourceName=None
      resourceObject=null
      tiling=NoTile
      imageSize=<0 0>
      margin=<ltrb 0 0 0 0>
      outlineSettings {
       bUseBrushTransparency=false
       roundingType=FixedRadius
       width=1
       color {colorUseRule=UseColor_ColorTable, specifiedColor=<rgba 0 0 0 0>}
       cornerRadii=<4 4 4 4>
      }
      tintColor {colorUseRule=UseColor_Specified, specifiedColor=<rgba 0.0395 0.0395 0.0395 1>}
      uVRegion {bIsValid=false, max=<0 0>, min=<0 0>}
     }
     disabledForeground {colorUseRule=UseColor_ColorTable, specifiedColor=<rgba 0 0 0 0>}
     hovered {
      bIsDynamicallyLoaded=false
      drawAs=RoundedBox
      imageType=NoImage
      mirroring=NoMirror
      resourceName=None
      resourceObject=null
      tiling=NoTile
      imageSize=<0 0>
      margin=<ltrb 0 0 0 0>
      outlineSettings {
       bUseBrushTransparency=false
       roundingType=FixedRadius
       width=1
       color {colorUseRule=UseColor_ColorTable, specifiedColor=<rgba 0 0 0 0>}
       cornerRadii=<4 4 4 4>
      }
      tintColor {colorUseRule=UseColor_Specified, specifiedColor=<rgba 0.0953 0.0953 0.0953 1>}
      uVRegion {bIsValid=false, max=<0 0>, min=<0 0>}
     }
     hoveredForeground {colorUseRule=UseColor_ColorTable, specifiedColor=<rgba 0 0 0 0>}
     hoveredSlateSound {resourceObject=null}
     normal {
      bIsDynamicallyLoaded=false
      drawAs=RoundedBox
      imageType=NoImage
      mirroring=NoMirror
      resourceName=None
      resourceObject=null
      tiling=NoTile
      imageSize=<0 0>
      margin=<ltrb 0 0 0 0>
      outlineSettings {
       bUseBrushTransparency=false
       roundingType=FixedRadius
       width=1
       color {colorUseRule=UseColor_ColorTable, specifiedColor=<rgba 0 0 0 0>}
       cornerRadii=<4 4 4 4>
      }
      tintColor {colorUseRule=UseColor_Specified, specifiedColor=<rgba 0.0395 0.0395 0.0395 1>}
      uVRegion {bIsValid=false, max=<0 0>, min=<0 0>}
     }
     normalForeground {colorUseRule=UseColor_ColorTable, specifiedColor=<rgba 0 0 0 0>}
     normalPadding=<ltrb 12 1.5 12 1.5>
     pressed {
      bIsDynamicallyLoaded=false
      drawAs=RoundedBox
      imageType=NoImage
      mirroring=NoMirror
      resourceName=None
      resourceObject=null
      tiling=NoTile
      imageSize=<0 0>
      margin=<ltrb 0 0 0 0>
      outlineSettings {
       bUseBrushTransparency=false
       roundingType=FixedRadius
       width=1
       color {colorUseRule=UseColor_ColorTable, specifiedColor=<rgba 0 0 0 0>}
       cornerRadii=<4 4 4 4>
      }
      tintColor {colorUseRule=UseColor_Specified, specifiedColor=<rgba 0.0284 0.0284 0.0284 1>}
      uVRegion {bIsValid=false, max=<0 0>, min=<0 0>}
     }
     pressedForeground {colorUseRule=UseColor_ColorTable, specifiedColor=<rgba 0 0 0 0>}
     pressedPadding=<ltrb 12 2 12 0>
     pressedSlateSound {resourceObject=null}
    }}
   slot {class=HorizontalBoxSlot, properties {horizontalAlignment=HAlign_Left, size {sizeRule=Automatic, value=0.5}}}
  }
  ScaleBox {
   widgetClassPath=&ScaleBox
   parent=GenerateCSV
   widget=ScaleBox
   slot {class=ButtonSlot, properties {horizontalAlignment=HAlign_Fill, verticalAlignment=VAlign_Fill, padding=<ltrb 0 0 0 0>}}
  }
  CommonTextBlock {
   widgetClassPath=&CommonTextBlock
   parent=ScaleBox
   widget=CommonTextBlock
   properties {
    style=&TextStyle-Base_C
    text=Generate CSV
    font {
     bForceMonospaced=false
     bMaterialIsStencil=false
     fontMaterial=null
     fontObject=&NotoSans
     letterSpacing=0
     monospacedWidth=1
     size=24
     skewAmount=0
     typefaceFontName=NotoSans-Regular
     outlineSettings {
      bApplyOutlineToDropShadows=false
      bMiteredCorners=false
      bSeparateFillAlpha=false
      outlineMaterial=null
      outlineSize=0
      outlineColor=<rgba 0 0 0 1>
     }
    }
    renderTransform {angle=0, scale=<1 1>, shear=<0 0>, translation=<0 -2.5>}
    shadowOffset=<0 0>
   }
   slot {class=ScaleBoxSlot}
  }
 ]
}
